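; Looping gameplay cue notify for the CoffeePuff hookshot projectile cast, providing affiliation-based cast audio.
; The application event only calls its parent, although its comment describes playing the cast sound and attaching it to the rope's projectile mesh.
; A helper function, not called in these graphs, spawns a sound on the player pawn's capsule, choosing the local variant when locally viewed, otherwise friendly or hostile by team affiliation.
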(event EventOnApplication(EitherVFX-preferthisone)
  (params (MyTarget :type Actor%20Object%20Reference) (Parameters :type Gameplay%20Cue%20Parameters%20Structure%20%28by%20ref%29) (ParticleComponents :type Array%20of%20FXSystem%20Component%20Object%20References) (AudioComponents :type Array%20of%20Audio%20Component%20Object%20References) (BurstCameraShakeInstance :type Camera%20Shake%20Base%20Object%20Reference) (BurstDecalInstance :type Decal%20Actor%20Object%20Reference))
  ;[0.02412 0.15 0.1436 0.5] OnApplication: Play the cast sound cue based on affiliation. Whenever the rope becomes available, attach it to the rope's projectile mesh component so that it moves along with the rope.
  (|CallParentFunction MyTarget Parameters ParticleComponents AudioComponents BurstCameraShakeInstance BurstDecalInstance))

(fn PlaySoundWithPlayerAffiliation
  (params (TargetActor :type Actor%20Object%20Reference) (LocalSoundCue :type Sound%20Base%20Object%20Reference) (FriendlySoundCue :type Sound%20Base%20Object%20Reference) (HostileSoundCue :type Sound%20Base%20Object%20Reference))
  ;[0.0206 0.15 0.1284 0.5] Spawn sound based on player affiliation.
  (Utilities|FlowControl|Sequence
    (:then_0
      (bind _asfort_player_pawn (Utilities|Casting|CastToFortPlayerPawn TargetActor))
      (bind _capsulecomponent (Variables|/Script/Engine.Character|GetCapsuleComponent _asfort_player_pawn)
        (:then
          (bind _returnvalue (Audio|SpawnSoundAttached (select (not (Pawn|IsLocallyViewed _asfort_player_pawn)) (Utilities|Select :Index (Team|GetTeamAffiliationforActortoLocalPlayer _asfort_player_pawn) :Friendly FriendlySoundCue :Hostile HostileSoundCue) LocalSoundCue) _capsulecomponent))
          (return _returnvalue))
        (:else)))
    (:then_1
      (return 0))))

(fn ConstructionScript)
